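; Blueprint for the Ashton pack Milo launcher, a chargeable rifle-class weapon that gains pierce-through at max charge.
; Charge start plays a player montage, charge FX, and audio shaped by a pitch and volume timeline.
; Max charge swaps the muzzle system and starts a self-recalling, delayed camera shake loop gated on the Charging flag.
; Release, firing and unequip end the charge, with firing adding a charged sound layer if max charge was reached.
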(event Custom|MultiStartCharge
  (bind _ambient_charge_audiocomponent (Variables|Default|GetAmbient_Charge_AudioComponent))
  ;[1 1 1] Start Charing
  (Variables|VFX|SetCharging true)
  (Variables|VFX|SetReached_Min_Charge false) ; Can play charged weapon fire sound
  (CallFunction|VFXSetUnchargedParams)
  (bind _asfort_player_pawn (Utilities|Casting|CastToFortPlayerPawn (Game|GetInstigator)))
  (Variables|Default|SetPlayerPawn _asfort_player_pawn)
  (Animation|Montage|PlayMontage (Variables|/Script/Engine.Character|GetMesh (Variables|Default|GetPlayerPawn)) "/Figment_Martin/Animation/Game/MainPlayer/Combat/Gadgets/Medium/Male/Ashton/Milo/Ashton_Milo_Canon_Charging_CMM_Montage.Ashton_Milo_Canon_Charging_CMM_Montage")
  ;;[1 1 1] Weapon charge audio setup
  (Audio|Components|Audio|FadeOut (Variables|Default|GetAmbient_Charge_AudioComponent) 0.3 0.0 "Linear" :!class /Script/Engine.AudioComponent)
  (CallFunction|PlayChargeAudio)
  (Components|Activation|Activate (Variables|Default|GetNS_ChargeFX) false)
  (Audio|Components|Audio|Play (Variables|Default|GetFullCharge_AudioComponent) 0.0)
  (bind (_direction _reload_pitch _volume) (Timeline|Reload_PItchChange Play)
    (:Update
      (Audio|Components|Audio|SetPitchMultiplier _ambient_charge_audiocomponent _reload_pitch)
      (Audio|Components|Audio|SetVolumeMultiplier _ambient_charge_audiocomponent _volume :!class /Script/Engine.AudioComponent))))

(event EventOnStartCharge
  ;[1 1 1] Start Charing
  (CallFunction|MultiStartCharge))

(event EventOnLocalClientReachedMaxCharge
  ;[1 1 1] EVENT -> Max Charge Reached
  (CallFunction|VFXSetChargedParams)
  (Variables|VFX|SetReached_Min_Charge true) ; Can play charged weapon fire sound
  (CallFunction|SetWeaponPierceThrough true 12)
  (Camera|PlayWorldCameraShake "/Game/Weapons/FORT_Rifles/Effects/Muzzle/B_Ashton_Milo_Shake_Charging.B_Ashton_Milo_Shake_Charging_C" (Transformation|GetActorLocation) (Variables|Default|GetShakeInnerRadius) (Variables|Default|GetCameraShakeOuterRadius))
  (Utilities|FlowControl|Delay 0.2)
  (if (Variables|VFX|GetCharging)
    (CallFunction|CameraShakeLoop)))

(event Custom|CameraShakeLoop
  ;[1 1 1] EVENT -> Max Charge Reached
  (Camera|PlayWorldCameraShake "/Game/Weapons/FORT_Rifles/Effects/Muzzle/B_Ashton_Milo_Shake_Charging.B_Ashton_Milo_Shake_Charging_C" (Transformation|GetActorLocation) (Variables|Default|GetShakeInnerRadius) (Variables|Default|GetCameraShakeOuterRadius))
  (Utilities|FlowControl|Delay 0.2)
  (if (Variables|VFX|GetCharging)
    (CallFunction|CameraShakeLoop)))

(event Custom|EndOfCharge
  ;[1 1 1] EVENT -> Max Charge Reached
  ;;[0 0.1358 0.15 0.5] Added here to prevent infinite screen shake, by breaking out of the loop once charge up is over
  (Variables|VFX|SetCharging false)
  (bind _chargeaudio (Variables|Default|GetChargeAudio)
    (:then
      (Audio|Components|Audio|FadeOut _chargeaudio 0.2 0.0 "Linear" :!class /Script/Engine.AudioComponent))
    (:else
      _))
  (Audio|Components|Audio|FadeOut (Variables|Default|GetFullCharge_AudioComponent) 0.2 0.0 "Linear" :!class /Script/Engine.AudioComponent)
  (Components|Activation|Deactivate (Variables|Default|GetNS_ChargeFX)))

(event EventOnChargeDown
  ;[1 1 1] EVENT -> Charge Down (mouse released)
  (CallFunction|EndOfCharge)
  (Utilities|FlowControl|Delay 0.034)
  (Variables|VFX|SetCharging false))

(event EventOnPlayWeaponFireFX
  (params (bPersistentFire :type Boolean) (bSecondaryFire :type Boolean))
  (bind _ambient_charge_audiocomponent (Variables|Default|GetAmbient_Charge_AudioComponent))
  ;[1 1 1] EVENT -> Play Weapon Fire FX
  (|CallParentFunction bPersistentFire false)
  (CallFunction|EndOfCharge)
  ;;[1 1 1] Fire Sound
  ;;;[1 1 1] Reset ambient volume and pitch
  (Audio|Components|Audio|SetPitchMultiplier _ambient_charge_audiocomponent 1)
  (Audio|Components|Audio|SetVolumeMultiplier _ambient_charge_audiocomponent 0.0 :!class /Script/Engine.AudioComponent)
  (Audio|Components|Audio|Stop (Variables|Default|GetFullCharge_AudioComponent))
  ;;;[1 1 1] If charged, play charged fire sound layer
  (if (Variables|VFX|GetReached_Min_Charge)
    (Audio|SpawnSoundAttached "/Figment_Martin/Athena/Sounds/LTMs/Ashton/Milo/Ashton_Milo_Secondary_Fire_Charge_Layer_Cue.Ashton_Milo_Secondary_Fire_Charge_Layer_Cue" _ambient_charge_audiocomponent) ; Play Charged Shot Layer
    (Variables|VFX|SetReached_Min_Charge false)))

(event FortWeapon|EventOnLocalUnEquip
  ;[1 1 1] On Unequip - to fix a bug if you were charging, then switched weapons before firing
  (Variables|VFX|SetCharging false)
  (CallFunction|VFXSetUnchargedParams)
  (Components|Animation|SetPlayRate (Variables|Visual|GetWeaponMesh) (Variables|VFX|GetAnimationRate(Uncharged)))
  (CallFunction|EndOfCharge)
  (Components|Timeline|SetNewTime (Variables|B_Rifle_AshtonPack_Milo_Launcher|GetReload_PItchChange) 0.0))

(fn PlayChargeAudio
  (bind _chargeaudio (Variables|Default|GetChargeAudio)
    (:then
      (if (Audio|Components|Audio|IsPlaying _chargeaudio)
        _
        (else
          _)))
    (:else
      _))
  (bind _returnvalue (Audio|SpawnSoundAttached "/Game/Athena/Sounds/LTMs/Ashton/Milo/Ashton_Milo_Secondary_Charge_Cue.Ashton_Milo_Secondary_Charge_Cue" (Variables|Visual|GetWeaponMesh)))
  (Variables|Default|SetChargeAudio _returnvalue))

(fn ConstructionScript
  (|CallParentFunction))

(fn VFX_SetChargedParams
  (Variables|Effects|Muzzle|SetMuzzle%20System (Variables|VFX|GetMuzzleFlash(Charged)))
  (return))

(fn VFX_SetUnchargedParams
  (Variables|Effects|Muzzle|SetMuzzle%20System (Variables|VFX|GetMuzzleFlash(Uncharged)))
  (return))
